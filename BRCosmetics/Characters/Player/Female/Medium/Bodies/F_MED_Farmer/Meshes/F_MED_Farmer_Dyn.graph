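# Physics asset for the F_MED_Farmer female body cosmetic, covering two dangling grenades (front and back) on the left side.
# Each pin body is constrained to its grenade, and each grenade to a tiny MSTR anchor capsule, with limited swing and twist.
# Both anchors and the left thigh connect to the pelvis through angularly free constraints.
# All bodies are capsules, and the thigh's is the largest.
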PhysicsAsset /BRCosmetics/Characters/Player/Female/Medium/Bodies/F_MED_Farmer/Meshes/F_MED_Farmer_Dyn
bodies {
 dyn_grenade_bk_l [{shapeType=Capsule, length=9.7188, radius=3.4062, center=<7 0 2.8631>, rotation=<rot 90 0 0>}]
 dyn_grenade_bk_l_MSTR [{shapeType=Capsule, length=0.505, radius=0.505, rotation=<rot 0 -90 0>}]
 dyn_grenade_bk_pin_l [{shapeType=Capsule, length=3, radius=2, center=<1.7301 -0.0001 0>, rotation=<rot 90 0 0>}]
 dyn_grenade_fr_l [{shapeType=Capsule, length=15.7812, radius=3, center=<3.7894 0 3.2198>, rotation=<rot 90 0 0>}]
 dyn_grenade_fr_l_MSTR [{shapeType=Capsule, length=0.505, radius=0.505, rotation=<rot 0 -90 0>}]
 dyn_grenade_fr_pin_l [{shapeType=Capsule, length=3, radius=2, center=<1.7301 -0.0001 0>, rotation=<rot 90 0 0>}]
 pelvis [{shapeType=Capsule, length=3, radius=2, rotation=<rot 90 0 0>}]
 thigh_l [{shapeType=Capsule, length=24.5749, radius=8.51, center=<-15.796 -0.0004 0.0713>, rotation=<rot 86.0791 180 180>}]
}
constraints [
 {bone1Name=dyn_grenade_fr_pin_l, bone2Name=dyn_grenade_fr_l, swing1LimitDegrees=20, swing2LimitDegrees=45, twistLimitDegrees=10}
 {bone1Name=dyn_grenade_bk_pin_l, bone2Name=dyn_grenade_bk_l, swing1LimitDegrees=20, swing2LimitDegrees=45, twistLimitDegrees=10}
 {bone1Name=dyn_grenade_fr_l, bone2Name=dyn_grenade_fr_l_MSTR, swing1LimitDegrees=45, swing2LimitDegrees=30, twistLimitDegrees=5}
 {bone1Name=dyn_grenade_bk_l, bone2Name=dyn_grenade_bk_l_MSTR, swing1LimitDegrees=45, swing2LimitDegrees=30, twistLimitDegrees=5}
 {
  bone1Name=dyn_grenade_fr_l_MSTR
  bone2Name=pelvis
  swing1LimitDegrees=45
  swing1Motion=Free
  swing2LimitDegrees=45
  swing2Motion=Free
  twistLimitDegrees=45
  twistMotion=Free
 }
 {
  bone1Name=dyn_grenade_bk_l_MSTR
  bone2Name=pelvis
  swing1LimitDegrees=45
  swing1Motion=Free
  swing2LimitDegrees=45
  swing2Motion=Free
  twistLimitDegrees=45
  twistMotion=Free
 }
 {
  bone1Name=thigh_l
  bone2Name=pelvis
  swing1LimitDegrees=45
  swing1Motion=Free
  swing2LimitDegrees=45
  swing2Motion=Free
  twistLimitDegrees=45
  twistMotion=Free
 }
]
