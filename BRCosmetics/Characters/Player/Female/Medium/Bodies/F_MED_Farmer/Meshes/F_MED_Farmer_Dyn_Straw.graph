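# Physics asset for the dynamic straw pieces on the female medium Farmer cosmetic body, covering arms, shoulders, chest and back.
# Small capsule bodies sit on each straw bone, each constrained to a master bone for its cluster with uniform swing and twist limit angles.
# The master bodies in turn constrain to a pelvis capsule with free swing and twist, and the left arm has separate inner and outer clusters.
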PhysicsAsset /BRCosmetics/Characters/Player/Female/Medium/Bodies/F_MED_Farmer/Meshes/F_MED_Farmer_Dyn_Straw
bodies {
 dyn_straw_arm_in_l_MSTR [{shapeType=Capsule, length=0.5, radius=0.5, rotation=<rot 90 0 0>}]
 dyn_straw_arm_l_01 [{shapeType=Capsule, length=3, radius=1, center=<0.1 0 0>, rotation=<rot 90 0 0>}]
 dyn_straw_arm_l_02 [{shapeType=Capsule, length=3, radius=1, center=<0.2 0 0>, rotation=<rot 90 0 0>}]
 dyn_straw_arm_l_03 [{shapeType=Capsule, length=3, radius=1, center=<0.1 0 0>, rotation=<rot 90 0 0>}]
 dyn_straw_arm_l_04 [{shapeType=Capsule, length=3, radius=1, center=<0.2 0 0>, rotation=<rot 90 0 0>}]
 dyn_straw_arm_out_l_MSTR [{shapeType=Capsule, length=0.5, radius=0.5, rotation=<rot 90 0 0>}]
 dyn_straw_arm_r_01 [{shapeType=Capsule, length=3, radius=1, center=<0.1 0 0>, rotation=<rot 90 0 0>}]
 dyn_straw_arm_r_02 [{shapeType=Capsule, length=3, radius=1, center=<0.2 0 0>, rotation=<rot 90 0 0>}]
 dyn_straw_arm_r_MSTR [{shapeType=Capsule, length=0.5, radius=0.5, rotation=<rot 90 0 0>}]
 dyn_straw_back_01 [{shapeType=Capsule, length=3, radius=1, center=<0.1 0 0>, rotation=<rot 90 0 0>}]
 dyn_straw_back_02 [{shapeType=Capsule, length=3, radius=1, center=<0.2 0 0>, rotation=<rot 90 0 0>}]
 dyn_straw_back_MSTR [{shapeType=Capsule, length=0.5, radius=0.5, rotation=<rot 90 0 0>}]
 dyn_straw_chest_01 [{shapeType=Capsule, length=3, radius=1, center=<0.1 0 0>, rotation=<rot 90 0 0>}]
 dyn_straw_chest_MSTR [{shapeType=Capsule, length=0.5, radius=0.5, rotation=<rot 90 0 0>}]
 dyn_straw_shoulder_l_01 [{shapeType=Capsule, length=3, radius=1, center=<0.1 0 0>, rotation=<rot 90 0 0>}]
 dyn_straw_shoulder_l_02 [{shapeType=Capsule, length=3, radius=1, center=<0.2 0 0>, rotation=<rot 90 0 0>}]
 dyn_straw_shoulder_l_03 [{shapeType=Capsule, length=3, radius=1, center=<0.05 0 0>, rotation=<rot 90 0 0>}]
 dyn_straw_shoulder_l_MSTR [{shapeType=Capsule, length=0.5, radius=0.5, rotation=<rot 90 0 0>}]
 dyn_straw_shoulder_r_01 [{shapeType=Capsule, length=3, radius=1, center=<0.05 0 0>, rotation=<rot 90 0 0>}]
 dyn_straw_shoulder_r_02 [{shapeType=Capsule, length=3, radius=1, center=<0.2 0 0>, rotation=<rot 90 0 0>}]
 dyn_straw_shoulder_r_03 [{shapeType=Capsule, length=3, radius=1, center=<0.1 0 0>, rotation=<rot 90 0 0>}]
 dyn_straw_shoulder_r_MSTR [{shapeType=Capsule, length=0.5, radius=0.5, rotation=<rot 90 0 0>}]
 pelvis [{shapeType=Capsule, length=3, radius=2, rotation=<rot 90 0 0>}]
}
constraints [
 {bone1Name=dyn_straw_arm_r_02, bone2Name=dyn_straw_arm_r_MSTR, swing1LimitDegrees=45, swing2LimitDegrees=45, twistLimitDegrees=45}
 {bone1Name=dyn_straw_arm_r_01, bone2Name=dyn_straw_arm_r_MSTR, swing1LimitDegrees=45, swing2LimitDegrees=45, twistLimitDegrees=45}
 {bone1Name=dyn_straw_shoulder_r_01, bone2Name=dyn_straw_shoulder_r_MSTR, swing1LimitDegrees=45, swing2LimitDegrees=45, twistLimitDegrees=45}
 {bone1Name=dyn_straw_shoulder_r_02, bone2Name=dyn_straw_shoulder_r_MSTR, swing1LimitDegrees=45, swing2LimitDegrees=45, twistLimitDegrees=45}
 {bone1Name=dyn_straw_shoulder_r_03, bone2Name=dyn_straw_shoulder_r_MSTR, swing1LimitDegrees=45, swing2LimitDegrees=45, twistLimitDegrees=45}
 {bone1Name=dyn_straw_chest_01, bone2Name=dyn_straw_chest_MSTR, swing1LimitDegrees=45, swing2LimitDegrees=45, twistLimitDegrees=45}
 {bone1Name=dyn_straw_back_02, bone2Name=dyn_straw_back_MSTR, swing1LimitDegrees=45, swing2LimitDegrees=45, twistLimitDegrees=45}
 {bone1Name=dyn_straw_back_01, bone2Name=dyn_straw_back_MSTR, swing1LimitDegrees=45, swing2LimitDegrees=45, twistLimitDegrees=45}
 {bone1Name=dyn_straw_shoulder_l_01, bone2Name=dyn_straw_shoulder_l_MSTR, swing1LimitDegrees=45, swing2LimitDegrees=45, twistLimitDegrees=45}
 {bone1Name=dyn_straw_shoulder_l_02, bone2Name=dyn_straw_shoulder_l_MSTR, swing1LimitDegrees=45, swing2LimitDegrees=45, twistLimitDegrees=45}
 {bone1Name=dyn_straw_shoulder_l_03, bone2Name=dyn_straw_shoulder_l_MSTR, swing1LimitDegrees=45, swing2LimitDegrees=45, twistLimitDegrees=45}
 {bone1Name=dyn_straw_arm_l_01, bone2Name=dyn_straw_arm_in_l_MSTR, swing1LimitDegrees=45, swing2LimitDegrees=45, twistLimitDegrees=45}
 {bone1Name=dyn_straw_arm_l_02, bone2Name=dyn_straw_arm_in_l_MSTR, swing1LimitDegrees=45, swing2LimitDegrees=45, twistLimitDegrees=45}
 {bone1Name=dyn_straw_arm_l_03, bone2Name=dyn_straw_arm_out_l_MSTR, swing1LimitDegrees=45, swing2LimitDegrees=45, twistLimitDegrees=45}
 {bone1Name=dyn_straw_arm_l_04, bone2Name=dyn_straw_arm_out_l_MSTR, swing1LimitDegrees=45, swing2LimitDegrees=45, twistLimitDegrees=45}
 {
  bone1Name=dyn_straw_arm_out_l_MSTR
  bone2Name=pelvis
  swing1LimitDegrees=45
  swing1Motion=Free
  swing2LimitDegrees=45
  swing2Motion=Free
  twistLimitDegrees=45
  twistMotion=Free
 }
 {
  bone1Name=dyn_straw_arm_in_l_MSTR
  bone2Name=pelvis
  swing1LimitDegrees=45
  swing1Motion=Free
  swing2LimitDegrees=45
  swing2Motion=Free
  twistLimitDegrees=45
  twistMotion=Free
 }
 {
  bone1Name=dyn_straw_shoulder_l_MSTR
  bone2Name=pelvis
  swing1LimitDegrees=45
  swing1Motion=Free
  swing2LimitDegrees=45
  swing2Motion=Free
  twistLimitDegrees=45
  twistMotion=Free
 }
 {
  bone1Name=dyn_straw_arm_r_MSTR
  bone2Name=pelvis
  swing1LimitDegrees=45
  swing1Motion=Free
  swing2LimitDegrees=45
  swing2Motion=Free
  twistLimitDegrees=45
  twistMotion=Free
 }
 {
  bone1Name=dyn_straw_shoulder_r_MSTR
  bone2Name=pelvis
  swing1LimitDegrees=45
  swing1Motion=Free
  swing2LimitDegrees=45
  swing2Motion=Free
  twistLimitDegrees=45
  twistMotion=Free
 }
 {
  bone1Name=dyn_straw_chest_MSTR
  bone2Name=pelvis
  swing1LimitDegrees=45
  swing1Motion=Free
  swing2LimitDegrees=45
  swing2Motion=Free
  twistLimitDegrees=45
  twistMotion=Free
 }
 {
  bone1Name=dyn_straw_back_MSTR
  bone2Name=pelvis
  swing1LimitDegrees=45
  swing1Motion=Free
  swing2LimitDegrees=45
  swing2Motion=Free
  twistLimitDegrees=45
  twistMotion=Free
 }
]
