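# Anim Blueprint for the female medium Farmer body cosmetic, copying the attached parent mesh's pose and adding secondary motion to outfit pieces.
# Weighted multi-bone constraints drive straw master bones and the left knee pad from nearby body bones.
# Crouch events toggle an interpolated alpha that applies an additive offset to the front-left grenade bone.
# Three chained RigidBody passes then simulate separate physics assets in spine-relative space, with lighter gravity for straw.
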AnimBlueprint /BRCosmetics/Characters/Player/Female/Medium/Bodies/F_MED_Farmer/Meshes/F_MED_Farmer_AnimBP
target_skeleton=&/F_MED_Farmer_Skeleton
graphs [
 AnimGraph {nodes [
   n1 = "Constraint - dyn_straw_chest_MSTR (PPP)"!(
    ComponentPose=Local To Component(LocalPose=Copy Pose From Mesh() {bUseAttachedParent=true})
    Alpha=1
    ConstraintWeights_0=0.0013
    ConstraintWeights_1:float=0.8393
    ConstraintWeights_2:float=0.1593
   ) -> (Pose) {
    alphaBoolBlend {blendOption=Linear}
    boneToModify {boneName=dyn_straw_chest_MSTR}
    constraintSetup [{targetBone {boneName=clavicle_r}}, {targetBone {boneName=spine_04}}, {targetBone {boneName=spine_05}}]
    constraintWeights [1,1,1]
   }
   n2 = "Constraint - dyn_straw_back_MSTR (PPPP)"!(
    ComponentPose=@n1
    Alpha=1
    ConstraintWeights_0=0.491
    ConstraintWeights_1:float=0.0069
    ConstraintWeights_2:float=0.48
    ConstraintWeights_3:float=0.0221
   ) -> (Pose) {
    alphaBoolBlend {blendOption=Linear}
    boneToModify {boneName=dyn_straw_back_MSTR}
    constraintSetup [{targetBone {boneName=clavicle_r}}, {targetBone {boneName=spine_04}}, {targetBone {boneName=spine_05}}, {targetBone {boneName=clavicle_l}}]
    constraintWeights [1,1,1,1]
   }
   n3 = "Constraint - dyn_straw_arm_r_MSTR (PPP)"!(ComponentPose=@n2, Alpha=1, ConstraintWeights_0=0.7306, ConstraintWeights_1:float=0.1254, ConstraintWeights_2:float=0.144) -> (Pose) {
    alphaBoolBlend {blendOption=Linear}
    boneToModify {boneName=dyn_straw_arm_r_MSTR}
    constraintSetup [{targetBone {boneName=upperarm_twist_01_r}}, {targetBone {boneName=upperarm_r}}, {targetBone {boneName=upperarm_twist_02_r}}]
    constraintWeights [1,1,1]
   }
   n4 = "Constraint - dyn_straw_shoulder_r_MSTR (PPP)"!(ComponentPose=@n3, Alpha=1, ConstraintWeights_0=0.7148, ConstraintWeights_1:float=0.2677, ConstraintWeights_2:float=0.0175) -> (Pose) {
    alphaBoolBlend {blendOption=Linear}
    boneToModify {boneName=dyn_straw_shoulder_r_MSTR}
    constraintSetup [{targetBone {boneName=clavicle_r}}, {targetBone {boneName=upperarm_twist_01_r}}, {targetBone {boneName=spine_05}}]
    constraintWeights [1,1,1]
   }
   n5 = "Constraint - dyn_straw_shoulder_l_MSTR (PPP)"!(ComponentPose=@n4, Alpha=1, ConstraintWeights_0=0.2784, ConstraintWeights_1:float=0.0013, ConstraintWeights_2:float=0.7202) -> (Pose) {
    alphaBoolBlend {blendOption=Linear}
    boneToModify {boneName=dyn_straw_shoulder_l_MSTR}
    constraintSetup [{targetBone {boneName=upperarm_twist_01_l}}, {targetBone {boneName=spine_05}}, {targetBone {boneName=clavicle_l}}]
    constraintWeights [1,1,1]
   }
   n6 = "Constraint - dyn_straw_arm_out_l_MSTR (PPP)"!(ComponentPose=@n5, Alpha=1, ConstraintWeights_0=0.1382, ConstraintWeights_1:float=0.1198, ConstraintWeights_2:float=0.742) -> (Pose) {
    alphaBoolBlend {blendOption=Linear}
    boneToModify {boneName=dyn_straw_arm_out_l_MSTR}
    constraintSetup [{targetBone {boneName=upperarm_twist_02_l}}, {targetBone {boneName=upperarm_l}}, {targetBone {boneName=upperarm_twist_01_l}}]
    constraintWeights [1,1,1]
   }
   n7 = "Constraint - dyn_straw_arm_in_l_MSTR (PPP)"!(ComponentPose=@n6, Alpha=1, ConstraintWeights_0=0.1378, ConstraintWeights_1:float=0.1194, ConstraintWeights_2:float=0.7428) -> (Pose) {
    alphaBoolBlend {blendOption=Linear}
    boneToModify {boneName=dyn_straw_arm_in_l_MSTR}
    constraintSetup [{targetBone {boneName=upperarm_twist_02_l}}, {targetBone {boneName=upperarm_l}}, {targetBone {boneName=upperarm_twist_01_l}}]
    constraintWeights [1,1,1]
   }
   n8 = "Constraint - knee_pad_l (PP)"!(ComponentPose=@n7, Alpha=1, ConstraintWeights_0=0.5, ConstraintWeights_1:float=0.5) -> (Pose) {
    alphaBoolBlend {blendOption=Linear}
    boneToModify {boneName=knee_pad_l}
    constraintSetup [{targetBone {boneName=thigh_l}}, {targetBone {boneName=calf_l}}]
    constraintWeights [1,1]
   }
   n9 = "Transform (Modify) Bone - Bone: dyn_grenade_fr_l_MSTR"(
    ComponentPose=@n8
    Alpha=Get crouching alpha() -> (crouching alpha:double)
    Translation="0.631,-3.204,2.407"
    Rotation="33.618,-8.075,-23.371"
    Scale="1,1,1"
   ) {
    rotationMode=BMM_Additive
    rotationSpace=BCS_BoneSpace
    translationMode=BMM_Additive
    translationSpace=BCS_BoneSpace
    alphaBoolBlend {blendOption=Linear}
    alphaScaleBiasClamp {bInterpResult=true}
    boneToModify {boneName=dyn_grenade_fr_l_MSTR}
    rotation=<rot 33.618 -8.075 -23.3709>
    translation=<0.631 -3.204 2.407>
   }
   n10 = RigidBody(ComponentPose=@n9) {
    bClampLinearTranslationLimitToRefPose=true
    bOverrideWorldGravity=true
    overridePhysicsAsset=&/F_MED_Farmer_Dyn_Straw
    simulationSpace=BaseBoneSpace
    alphaBoolBlend {blendOption=Linear}
    baseBoneRef {boneName=spine_02}
    componentAppliedLinearAccClamp=<2000 2000 2000>
    componentLinearAccScale=<1 1 1>
    componentLinearVelScale=<1 1 1>
    overrideWorldGravity {z=-200}
   }
   n11 = RigidBody(ComponentPose=@n10) {
    bClampLinearTranslationLimitToRefPose=true
    bOverrideWorldGravity=true
    overridePhysicsAsset=&/F_MED_Farmer_Dyn
    simulationSpace=BaseBoneSpace
    alphaBoolBlend {blendOption=Linear}
    baseBoneRef {boneName=spine_01}
    componentAppliedLinearAccClamp=<2000 2000 2000>
    componentLinearAccScale=<0.8 0.5 0.5>
    componentLinearVelScale=<0.8 0.5 0.5>
    overrideWorldGravity {z=-980}
   }
   n12 = RigidBody(ComponentPose=@n11) {
    bClampLinearTranslationLimitToRefPose=true
    bOverrideWorldGravity=true
    overridePhysicsAsset=&/F_MED_Farmer_Dyn_legs
    simulationSpace=BaseBoneSpace
    alphaBoolBlend {blendOption=Linear}
    baseBoneRef {boneName=spine_01}
    componentAppliedLinearAccClamp=<2000 2000 2000>
    componentLinearAccScale=<0.5 0.5 0.5>
    componentLinearVelScale=<0.5 0.5 0.5>
    overrideWorldGravity {z=-980}
   }
   Output Pose(Result=Component To Local(ComponentPose=@n12)) {layerGroup=None}
  ]}
 EventGraph {nodes [
   Event OnBeginCrouching()
   > Set crouching alpha(crouching alpha:double=1) -> (Output_Get:double)
   Event OnEndCrouching()
   > Set crouching alpha(crouching alpha=0)
   Event BlueprintUpdateAnimation()
  ]}
]
variables [crouching alpha {type=number}]
